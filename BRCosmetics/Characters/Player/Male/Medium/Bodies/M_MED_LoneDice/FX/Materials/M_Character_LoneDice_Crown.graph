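# Translucent unlit Niagara sprite material that draws the crown effect for the LoneDice character.
# Vertex-offset math collapses each quad to its particle position and rebuilds it around an adjustable pivot, facing the camera along a view-plus-instance axis.
# A masked two-color tint, blendable to particle color, gets a color-dodged, view-aligned reflection band before emissive adjustment.
# The grayscale crown texture masks the reflection and, with particle alpha, shapes opacity.
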Material /BRCosmetics/Characters/Player/Male/Medium/Bodies/M_MED_LoneDice/FX/Materials/M_Character_LoneDice_Crown
graph {
 properties {
  bAutomaticallySetUsageInEditor=false
  bUsedWithNiagaraSprites=true
  blendMode=BLEND_Translucent
  shadingModel=MSM_Unlit
  usedShadingModels=Unlit
 }
 nodes [
  n1 = MaterialFunctionCall(Float2=TextureCoordinate()) {materialFunction=&BreakOutFloat2Components}
  n2 = ScalarParameter() {defaultValue=1.25, parameterName=Pivot Position Y}
  n3 = Multiply(
   A=MaterialFunctionCall(X=Subtract(A=@n1.R) {constB=0.5}, Y=Subtract(A=@n1.G, B=@n2)) {materialFunction=&MakeFloat2}.Result
   B=ParticleSize()
  ) {constB=44}
  n4 = ComponentMask(None=@n3) {r=true}
  n5 = Normalize(VectorInput=Subtract(A=CameraPositionWS(), B=ParticlePositionWS()))
  n6 = VectorParameter() {parameterName=Particle Axis, defaultValue=<rgba 0.3 -0.5 1 0>}
  n7 = Add(
   A=Transform(None=VectorParameter() {parameterName=View Axis}.RGB) {transformSourceType=TRANSFORMSOURCE_View}
   B=Transform(None=Normalize(VectorInput=@n6.RGB)) {transformSourceType=TRANSFORMSOURCE_Instance}
  )
  n8 = Normalize(VectorInput=CrossProduct(A=@n5, B=Normalize(VectorInput=@n7))) {desc=X}
  n9 = ComponentMask(None=@n3) {g=true}
  n10 = Normalize(VectorInput=CrossProduct(A=@n5, B=@n8)) {desc=Y}
  Add(A=Subtract(A=ParticlePositionWS(), B=WorldPosition().XYZ) {desc=Collapse}, B=Add(A=Multiply(A=@n4, B=@n8), B=Multiply(A=@n9, B=@n10)))
  n11 = ParticleColor()
  n12 = VectorParameter() {parameterName=Color2, defaultValue=<rgba 0.2135 0.1019 0.0434 1>}
  n13 = VectorParameter() {parameterName=Color1, defaultValue=<rgba 0.6795 0.3278 0.0666 1>}
  n14 = SphereMask(A=TextureCoordinate() {vTiling=0}, B=Constant2Vector() {r=0.5}, Radius=Constant() {r=0.35}, Hardness=Constant())
  n15 = LinearInterpolate(
   A=Multiply(A=@n11.RGB, B=LinearInterpolate(A=@n12.RGB, B=@n13.RGB, Alpha=@n14))
   B=@n11.RGB
   Alpha=ScalarParameter() {parameterName=UseParticleColor}
  )
  n16 = TextureSample(UVs=MaterialFunctionCall!(ReflectionVector=CameraVectorWS(), ReflectionTexture) {materialFunction=/Engine/Functions/Engine_MaterialFunctions01/Reflections/ViewAlignedReflection}.UVs) {texture=/Game/Global/Textures/Default/Masks_and_Gradients/gradient_stasis}
  n17 = ScalarParameter() {defaultValue=16, parameterName=ReflectionbandPower}
  n18 = Saturate(None=Multiply(A=Power(Base=@n16.R, Exp=@n17) {constExponent=4}, B=ScalarParameter() {defaultValue=1, parameterName=ReflectionbandStrenght}))
  n19 = Constant3Vector() {constant=<rgba 1 1 1 1>}
  n20 = TextureSampleParameter2D() {
   parameterName=TextureInput
   samplerType=SAMPLERTYPE_Grayscale
   texture=&/T_LoneDice_C_Crown
  }
  n21 = Multiply(A=Multiply(A=Multiply(A=@n18, B=@n19), B=ScalarParameter() {defaultValue=0.4, parameterName=ReflectionStrenght}), B=@n20.R)
  Add(A=@n15, B=@n21)
  MaterialFunctionCall(Emissive=MaterialFunctionCall!(Blend=@n21, Base=@n15) {materialFunction=/Engine/Functions/Engine_MaterialFunctions03/Blends/Blend_ColorDodge}.Result) {materialFunction=&MF_FX_EmissiveAdjustments}
  Multiply(A=Subtract(A=Power(Base=@n20.R) {constExponent=0.5}) {constB=0.1}, B=ParticleColor().A)
 ]
}
