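# Niagara system for the LoneDice character cosmetic that renders a sprite Crown emitter attached to a sampled socket.
# Burst-spawned particles persist and follow the socket, with BigHeadMode increasing their offset and size.
# The dot product of camera vector and owner's up axis blends crown orientation and camera offset to adapt to viewing angle.
# Time of day tints the color, and User.bisHeadHidden sets the particle visibility tag.
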NiagaraSystem /BRCosmetics/Characters/Player/Male/Medium/Bodies/M_MED_LoneDice/FX/NS_Character_LoneDice
dependencies {
 usedDataInterfaces [&NiagaraDataInterfaceCamera, &NiagaraDataInterfaceCurve, &NiagaraDataInterfaceSocketReader]
 usedDynamicInputs []
 usedModules [
  &self:SystemSpawnScript.SystemScriptSource.SystemScriptGraph.NiagaraNodeAssignment_0.SetVariables_#1
  &SystemState
 ]
}
emitters {Crown {
  emitterUpdateScript [
   &EmitterState(Life Cycle Mode=::NewEnumerator1, Loop Behavior=::NewEnumerator1, Loop Duration Mode=::NewEnumerator1, UseLoopDelay=false)
   &SpawnBurst_Instantaneous(Use Spawn Probability=false, UseLoopCountLimit=false)
   /SetVariables(Emitter.CrownOrientation:Vector3f=&Lerp_Vector(
     A=<0.5 0 1>
     B=<0.5 -0.5 1>
     Alpha=&NormalizeFloat(
      Value=&DotProduct(
       Vector A=&ReturnCameraVector(Camera Vectors=::NewEnumerator1, Coordinate Space=::Simulation)
       Vector B=&TransformVector(Vector=$Engine.Owner.SystemZAxis, Source Space=::Simulation, Destination Space=::World)
      )
      Min=0
      Max=1
     )
    )) {isSetParameters=true}
  ]
  particleSpawnScript [&InitializeParticle(Lifetime=0, Sprite Size=<18 18>, Sprite Size Mode=::NewEnumerator4, UsePositionOffset=false, Write Material Random=true)]
  particleUpdateScript [
   &SocketLocation(
    Orientation Sampling=::NewEnumerator1
    SamplePosOffset=true
    Sampled Position Offset=&Lerp_Vector(A=<11 0 0>, B=<22 0 0>, Alpha=$NPC.FortniteNPC.BigHeadMode)
    Socket Reader=$System.Socket Reader
    Socket Sampling Mode=::NewEnumerator3
    VelocitySpeedLimit=false
   )
   &ParticleState(Kill Particles When Lifetime Has Elapsed=false, Let Infinitely Lived Particles Die When Emitter Deactivates=true)
   &ScaleSpriteSize(
    Scale Sprite Size Mode=::NewEnumerator0
    Uniform Scale Factor=&Lerp_Float(A=1, B=1.65, Alpha=$NPC.FortniteNPC.BigHeadMode)
   )
   &ScaleColor(
    Scale RGB=&MakeVectorFromLinearColorRGB(Color=&FNTimeOfDayColor(Night Color=<rgba 0.494 0.6808 1 1>, Morning Color=<rgba 1 1 1 1>, Day Color=<rgba 1 1 1 1>, Evening Color=<rgba 0.494 0.6808 1 1>))
    ScaleA=true
    ScaleRGB=true
   )
   &CameraOffset(Camera Offset Amount=&Lerp_Float(
     A=5
     B=0
     Alpha=&NormalizeFloat(
      Value=&DotProduct(
       Vector A=&ReturnCameraVector(Camera Vectors=::NewEnumerator1, Coordinate Space=::World)
       Vector B=$Engine.Owner.SystemZAxis
      )
      Min=0
      Max=1
     )
    ))
   /SetVariables(Particles.VisibilityTag=&MakeCustomIntFromBool(Bool=$User.bisHeadHidden, True INT=1, False INT=0)) {isSetParameters=true}
  ]
  rendererClasses [&NiagaraSpriteRendererProperties]
  renderers [{rendererClass=&NiagaraSpriteRendererProperties}]
 }}
summary {
 systemName=NS_Character_LoneDice
 emitters [{emitterName=Crown, rendererClasses [&NiagaraSpriteRendererProperties]}]
 userVariables [
  User.bisHeadHidden {type=NiagaraBool}
  User.Reactivity {type=NiagaraFloat, defaultValue=0, description=""}
 ]
}
