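# Masked unlit material for the Lone Dice character body's graffiti sprites: animated spray splats and drips on camera-facing quads.
# World position offset billboards each quad, shrinking sprites on the head and hiding them in sphere-masked zones.
# Vertex colours supply per-sprite randomness; time-sampled curve atlas rows drive splat and drip animation over noise and SubUV atlas frames.
# A reactivity parameter adds a stepped animation, larger sprites and colour shift.
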Material /BRCosmetics/Characters/Player/Male/Medium/Bodies/M_MED_LoneDice/Materials/M_LoneDice_GraffitiSprites
graph {
 properties {
  bCastRayTracedShadows=false
  bUseMaterialAttributes=true
  bUsedWithMeshDeformer=true
  bUsedWithMorphTargets=true
  bUsedWithSkeletalMesh=true
  blendMode=BLEND_Masked
  ditherOpacityMask=true
  pixelDepthOffsetMode=PDOM_Legacy
  shadingModel=MSM_Unlit
  usedShadingModels=Unlit
 }
 nodes [
  Constant() {r=0.1}
  Constant() {r=1}
  Constant() {r=0.152}
  Constant() {r=1}
  n1 = SetMaterialAttributes!(
   MaterialAttributes
   Emissive Color=NamedRerouteUsage()
   Opacity Mask=NamedRerouteUsage()
   World Position Offset=NamedRerouteUsage()
   Pixel Depth Offset
  ) {
   desc=Masked 120
   attributeSetTypes [
    #1
    #2
    #3
    #4
   ]
  }
  MaterialFunctionCall(In=@n1) {materialFunction=&MF_TechArtOverrides}
  n2 = Time()
  n3 = ScalarParameter() {defaultValue=0.2, desc=0.15, group=LoneDice Graffiti Sprites, parameterName=LD_Spray_Speed}
  n4 = LinearInterpolate(
   A=@n3
   B=LinearInterpolate(
    A=@n3
    B=ScalarParameter() {defaultValue=0.3, desc=0.3, group=LoneDice Graffiti Sprites, parameterName=LD_Spray_SpeedMax}
    Alpha=NamedRerouteUsage()
   )
   Alpha=ScalarParameter() {defaultValue=1, group=LoneDice Graffiti Sprites, parameterName=LD_Spray_SpeedRandomness}
  )
  n5 = LinearInterpolate(A=@n4, B=Multiply(A=@n4) {constB=2}, Alpha=NamedRerouteUsage())
  n6 = Multiply(A=@n2, B=@n5)
  n7 = NamedRerouteUsage()
  n8 = CurveAtlasRowParameter(CurveTime=Frac(None=Subtract(A=@n6, B=@n7))) {
   atlas=/BRCosmetics/Characters/Player/Male/Medium/Bodies/M_MED_LoneDice/Textures/CA_LoneDice
   curve=/BRCosmetics/Characters/Player/Male/Medium/Bodies/M_MED_LoneDice/Textures/C_LoneDice
   group=LoneDice Graffiti Sprites
   parameterName=LD_SprayCurve
  }
  MaterialFunctionCall(Number=NamedRerouteDeclaration(None=@n8.A) {name=SplatsGrowAnim, nodeColor=<rgba 1 0 0.8906 1>}) {materialFunction=&DebugScalarValues}
  n9 = Time()
  n10 = Constant2Vector() {g=-0.1, r=0.1}
  n11 = TextureCoordinate()
  n12 = Add(A=@n11) {constB=-0.5}
  n13 = TextureSample!(
   UVs=Add(
    A=Multiply(A=@n9, B=@n10) {constB=0.5}
    B=MaterialFunctionCall(Vector or UVs=@n12) {materialFunction=&VectorToRadialValue}.Radial Coordinates
   )
   Tex
   "DDX(UVs)"=DDX(Value=@n11)
   "DDY(UVs)"=DDY(Value=@n11)
   Apply View MipBias
  ) {
   mipValueMode=TMVM_Derivative
   samplerSource=SSM_Wrap_WorldGroupSettings
   texture=/BRCosmetics/Characters/Player/Male/Medium/Bodies/M_MED_LoneDice/Textures/T_LoneDice_Outline_Noise
  }
  n14 = Multiply(A=Saturate(None=@n13.G), B=ScalarParameter() {defaultValue=0.1, group=LoneDice Graffiti Sprites, parameterName=addd})
  n15 = NamedRerouteUsage()
  n16 = NamedRerouteDeclaration(None=LinearInterpolate(A=NamedRerouteUsage(), B=NamedRerouteUsage(), Alpha=NamedRerouteUsage()) {desc=anim}) {name=SplatAnimation, nodeColor=<rgba 1 0 0.1406 1>}
  Multiply(
   A=SmoothStep(
    Min=ScalarParameter() {defaultValue=0.37, group=LoneDice Graffiti Sprites, parameterName=mmmiu}
    Max=ScalarParameter() {defaultValue=1.4115, group=LoneDice Graffiti Sprites, parameterName=maxww}
    Value=Add(A=@n14, B=@n15)
   )
   B=@n16
  )
  n17 = MaterialFunctionCall(Float2=Subtract(A=TextureCoordinate()) {constB=0.5}) {materialFunction=&BreakOutFloat2Components}
  n18 = Constant3Vector() {constant=<rgba 1 0 0 0>}
  n19 = NamedRerouteUsage()
  n20 = LinearInterpolate(
   A=ScalarParameter() {defaultValue=28, desc=28, group=LoneDice Graffiti Sprites, parameterName=LD_Sprites_Size}
   B=ScalarParameter() {defaultValue=20, desc=20, group=LoneDice Graffiti Sprites, parameterName=LD_HeadSprites_Size}
   Alpha=SmoothStep(Value=ComponentMask(None=@n19) {b=true}) {constMax=160, constMin=140}
  ) {desc=smaller sprites on head}
  n21 = Multiply(A=@n20, B=Constant() {r=1.35}) {constB=1.5}
  n22 = NamedRerouteUsage()
  n23 = Constant3Vector() {constant=<rgba 0 3 79.4695 1>}
  n24 = Constant() {r=20}
  n25 = Constant3Vector() {constant=<rgba -2.1897 5.9875 151.183 1>}
  n26 = Constant() {r=11.3704}
  n27 = Add(
   A=Add(A=SphereMask(A=@n19, B=@n23, Radius=@n24), B=SphereMask(A=@n19, B=@n25, Radius=@n26))
   B=SphereMask(A=@n19, B=Constant3Vector() {constant=<rgba 26.0706 0 40.5451 1>}, Radius=Constant() {r=10.2412})
  )
  n28 = MaterialFunctionCall(Float2=LinearInterpolate(A=LinearInterpolate(A=@n20, B=@n21, Alpha=@n22) {desc=bigger sprites when reactive}, Alpha=@n27) {constB=0, desc=No sprites on crotch area+Ear}) {materialFunction=&BreakOutFloat2Components}
  n29 = Multiply(A=Multiply(A=@n17.R, B=Transform(None=@n18) {transformSourceType=TRANSFORMSOURCE_Camera}), B=@n28.R)
  n30 = Constant3Vector() {constant=<rgba 0 -1 0 0>}
  n31 = Multiply(A=Multiply(A=@n17.G, B=Transform(None=@n30) {transformSourceType=TRANSFORMSOURCE_Camera}), B=@n28.G)
  n32 = ScalarParameter() {defaultValue=-15, group=LoneDice Graffiti Sprites, parameterName=LD_CamOffset}
  NamedRerouteDeclaration(None=Add(
    A=Add(A=@n29, B=@n31)
    B=MaterialFunctionCall(Offset Amount=@n32) {materialFunction=/Engine/Functions/Engine_MaterialFunctions02/WorldPositionOffset/CameraOffset}.Clamped Camera Offset
   )) {name=WPO_OUT, nodeColor=<rgba 0 1 0.5078 1>}
  n33 = NamedRerouteUsage()
  NamedRerouteDeclaration(None=SphereMask(A=NamedRerouteUsage(), B=Constant() {r=3}, Radius=LinearInterpolate(Alpha=@n33) {constA=0.1, constB=0.75}) {attenuationRadius=0.1}) {name=Drips_Sprites, desc=sprites that play drip anim, nodeColor=<rgba 0.7578 1 0 1>}
  n34 = Multiply(A=NamedRerouteUsage(), B=Constant() {desc=8 pi, r=25.1327})
  n35 = ScalarParameter() {defaultValue=1, group=LoneDice Graffiti Sprites, parameterName=LD_Splats_RandomRot_Scale}
  n36 = Multiply(
   A=ScalarParameter() {defaultValue=1, group=LoneDice Graffiti Sprites, parameterName=LD_Splats_RandomSizeX, sliderMax=1}
   B=NamedRerouteUsage()
  )
  n37 = Multiply(
   A=ScalarParameter() {defaultValue=1, group=LoneDice Graffiti Sprites, parameterName=LD_Splats_RandomSizeY, sliderMax=1}
   B=NamedRerouteUsage()
  )
  n38 = NamedRerouteDeclaration(None=LinearInterpolate(
    A=ScalarParameter() {defaultValue=0.5, group=LoneDice Graffiti Sprites, parameterName=LD_Splats_GlobalSize, sliderMax=1}
    B=MaterialFunctionCall(X=@n36, Y=@n37) {materialFunction=&MakeFloat2}.Result
    Alpha=ScalarParameter() {defaultValue=0.5, group=LoneDice Graffiti Sprites, parameterName=LD_Splats_RandomSizeScale}
   ) {constA=1, desc=Random size}) {name=SplatsRandSize, nodeColor=<rgba 0.5 0 1 1>}
  MaterialExpressionMaterialFunctionCall_9 = MaterialFunctionCall(
   UVs=Rotator(Time=Multiply(A=@n34, B=@n35)) {desc=Random rotation}
   Texture Scale=LinearInterpolate(A=Multiply(A=@n38) {constB=0.3}, B=@n38, Alpha=NamedRerouteUsage()) {constA=0.5}
  ) {materialFunction=&ScaleUVsByCenter}
  MaterialExpressionConstant_0 = Constant() {r=2}
  MaterialExpressionNamedRerouteUsage_16 = NamedRerouteUsage()
  n39 = Custom!(uv=@MaterialExpressionMaterialFunctionCall_9.UVs, size=@MaterialExpressionConstant_0, frame=@MaterialExpressionNamedRerouteUsage_16) {
   code="return uv/size + float2(floor(fmod(frame,size)), floor(frame/size))/size;"
   description=SubUV
   outputType=CMOT_Float2
   inputs [
    {
     inputName=uv
     input {expression=&self:MaterialExpressionMaterialFunctionCall_9}
    }
    {
     inputName=size
     input {expression=&self:MaterialExpressionConstant_0}
    }
    {
     inputName=frame
     input {expression=&self:MaterialExpressionNamedRerouteUsage_16}
    }
   ]
  }
  n40 = TextureSample!(UVs=NamedRerouteDeclaration(None=@n39) {name=UVs_Splats, nodeColor=<rgba 1 0 0.2812 1>}, Tex, Bias, Apply View MipBias) {
   mipValueMode=TMVM_MipBias
   samplerSource=SSM_Clamp_WorldGroupSettings
   samplerType=SAMPLERTYPE_Masks
   texture=/BRCosmetics/Characters/Player/Male/Medium/Bodies/M_MED_LoneDice/Textures/T_LoneDice_Outline_SprayAtlas
  }
  NamedRerouteDeclaration(None=Multiply(A=NamedRerouteUsage(), B=@n40.R)) {name=SplatsMask, nodeColor=<rgba 0.8047 0 1 1>}
  NamedRerouteDeclaration(None=@n8.G) {name=NonReactiveSplatsAnim, nodeColor=<rgba 1 0 0.0937 1>}
  n41 = TextureSample!(UVs=TextureCoordinate(), Tex, Apply View MipBias) {
   samplerSource=SSM_Wrap_WorldGroupSettings
   texture=&T_LoneDice_Outline_Noise
  }
  n42 = NamedRerouteUsage()
  MaterialExpressionOneMinus_6 = OneMinus(None=Power(Base=Multiply(A=@n41.B, B=@n42)))
  n43 = TextureSample(UVs=NamedRerouteUsage()) {
   samplerSource=SSM_Wrap_WorldGroupSettings
   texture=&T_LoneDice_Outline_Noise
  }
  n44 = ComponentMask(None=TextureCoordinate()) {g=true}
  MaterialExpressionOneMinus_7 = OneMinus(None=Multiply(A=@n43.A, B=Step(X=@n44) {constY=0.5, desc=mask out the top half}))
  MaterialExpressionConstant_15 = Constant() {r=0.5}
  n45 = Custom!(a=@MaterialExpressionOneMinus_6, b=@MaterialExpressionOneMinus_7, k=@MaterialExpressionConstant_15) {
   code="float h = max( k-abs(a-b), 0.0 )/k;\r\nreturn min( a, b ) - h*h*k*(1.0/4.0);"
   desc=SmoothMin
   description=SmoothMin
   outputType=CMOT_Float1
   inputs [
    {
     inputName=a
     input {expression=&self:MaterialExpressionOneMinus_6}
    }
    {
     inputName=b
     input {expression=&self:MaterialExpressionOneMinus_7}
    }
    {
     inputName=k
     input {expression=&self:MaterialExpressionConstant_15}
    }
   ]
  }
  n46 = OneMinus(None=@n45)
  NamedRerouteDeclaration(None=Multiply(A=SmoothStep(Value=@n46) {constMax=0.5, constMin=0.4}, B=NamedRerouteUsage())) {name=DripMask, nodeColor=<rgba 0 0.6016 1 1>}
  n47 = ScalarParameter() {defaultValue=5, group=LoneDice Graffiti Sprites, parameterName=LD_ReactiveSteps}
  n48 = Multiply(A=Time(), B=@n47) {constB=0.5}
  n49 = Multiply(
   A=Divide(A=Round(None=@n48), B=@n47)
   B=ScalarParameter() {defaultValue=0.3, group=LoneDice Graffiti Sprites, parameterName=LD_Spray_Speed_Reactive}
  )
  n50 = Frac(None=Add(A=NamedRerouteUsage()) {constB=0.3312})
  n51 = CurveAtlasRowParameter(CurveTime=Frac(None=Subtract(A=@n49, B=@n50))) {
   atlas=&CA_LoneDice
   curve=&C_LoneDice
   group=LoneDice Graffiti Sprites
   parameterName=LD_SprayCurve
  }
  NamedRerouteDeclaration(None=Clamp(None=@n51.G)) {name=ReactiveSplatsAnim, nodeColor=<rgba 1 0.3134 0 1>}
  n52 = TextureCoordinate()
  n53 = Constant2Vector() {g=0.5, r=0.5}
  n54 = NamedRerouteUsage()
  n55 = LinearInterpolate(Alpha=@n54)
  n56 = LinearInterpolate(
   A=ScalarParameter() {defaultValue=0.15, parameterName=Start}
   B=ScalarParameter() {defaultValue=0.35, parameterName=End}
   Alpha=LinearInterpolate(A=SphereMask(A=@n52, B=@n53, Radius=@n55) {attenuationRadius=0.5, hardnessPercent=0}, Alpha=@n54)
  ) {constA=0.15, constAlpha=1, constB=0.35}
  n57 = Add(A=@n56, B=ScalarParameter() {defaultValue=0.025, parameterName=Hardness})
  n58 = NamedRerouteUsage()
  n59 = Add(A=Multiply(A=SmoothStep(Min=@n56, Max=@n57, Value=@n58), B=@n16), B=NamedRerouteUsage()) {constB=0, desc="spray = splat + drip"}
  n60 = Multiply(
   A=Add(A=NamedRerouteUsage()) {constB=0.5}
   B=ScalarParameter() {defaultValue=2, group=LoneDice Graffiti Sprites, parameterName=LD_SprayStrength}
  )
  NamedRerouteDeclaration(None=Saturate(None=Multiply(A=@n59, B=@n60))) {name=OpacityMask_OUT, nodeColor=<rgba 0.3828 0 1 1>}
  n61 = VectorParameter() {group=LoneDice Graffiti Sprites, parameterName=LD_GraffitiOutline_ReactiveColor, defaultValue=<rgba 1.1 0.1837 0.1155 1>}
  NamedRerouteDeclaration(None=LinearInterpolate(
    A=VectorParameter() {group=LoneDice Graffiti Sprites, parameterName=LD_GraffitiColor, defaultValue=<rgba 0 0 0 1>}.RGB
    B=@n61.RGB
    Alpha=NamedRerouteUsage()
   )) {name=EmissiveColor_OUT, nodeColor=<rgba 0.125 1 0 1>}
  n62 = SphereMask(A=NamedRerouteUsage(), B=Constant() {r=0.5}, Radius=Constant() {r=0.5}) {attenuationRadius=0.1}
  n63 = NamedRerouteUsage()
  NamedRerouteDeclaration(None=Clamp(None=LinearInterpolate(B=@n62, Alpha=@n63))) {name=ReactiveSprites, nodeColor=<rgba 1 0.4501 0 1>}
  n64 = MaterialFunctionCall(Float2=@MaterialExpressionMaterialFunctionCall_9.UVs) {materialFunction=&BreakOutFloat2Components}
  n65 = Constant() {r=0.5}
  NamedRerouteDeclaration(None=Multiply(A=SphereMask(A=@n64.R, B=@n65) {attenuationRadius=0.5}, B=SphereMask(A=@n64.G, B=@n65) {attenuationRadius=0.5})) {name=0-1Mask, desc=mask only uv 0-1, nodeColor=<rgba 1 0.7266 0 1>}
  n66 = Multiply(A=@n2, B=@n5)
  n67 = Modulo!(A=Floor(None=Subtract(A=@n66, B=@n7)), B=Constant() {r=3})
  n68 = Floor(None=Multiply(A=NamedRerouteUsage()) {constB=3})
  NamedRerouteDeclaration(None=Floor(None=Add(A=@n67, B=@n68))) {name=SplatsRandFrameNum, desc=assign rand 0-3 frame num, nodeColor=<rgba 0.9688 0 1 1>}
  n69 = NamedRerouteUsage()
  MaterialExpressionAdd_3 = Add(
   A=TextureCoordinate()
   B=MaterialFunctionCall(X=Constant(), Y=LinearInterpolate(Alpha=@n69) {constA=0.5, constB=-0.6}) {materialFunction=&MakeFloat2}.Result
  )
  MaterialExpressionConstant_20 = Constant() {r=2}
  MaterialExpressionRound_2 = Round(None=Multiply(A=NamedRerouteUsage(), B=Constant() {r=3}) {constB=4})
  n70 = Custom!(uv=@MaterialExpressionAdd_3, size=@MaterialExpressionConstant_20, frame=@MaterialExpressionRound_2) {
   code="return uv/size + float2(floor(fmod(frame,size)), floor(frame/size))/size;"
   description=SubUV
   outputType=CMOT_Float2
   inputs [
    {
     inputName=uv
     input {expression=&self:MaterialExpressionAdd_3}
    }
    {
     inputName=size
     input {expression=&self:MaterialExpressionConstant_20}
    }
    {
     inputName=frame
     input {expression=&self:MaterialExpressionRound_2}
    }
   ]
  }
  NamedRerouteDeclaration(None=@n70) {name=UVs_Drips, nodeColor=<rgba 1 0 0.7734 1>}
  n71 = VertexColor()
  NamedRerouteDeclaration(None=@n71) {name=Random, nodeColor=<rgba 0.9219 1 0 1>}
  NamedRerouteDeclaration(None=@n8.R) {name=DripsAnim, nodeColor=<rgba 1 0 0.375 1>}
  n72 = ScalarParameter() {group=LoneDice Graffiti Sprites, parameterName=LD_Reactivity, sliderMax=1}
  NamedRerouteDeclaration(None=Multiply(A=Floor(None=@n72), B=ScalarParameter() {defaultValue=1, group=LoneDice Graffiti Sprites, parameterName=LD_ReactivityToggle, sliderMax=1})) {name=Reactivity, nodeColor=<rgba 0 0.2031 1 1>}
  n73 = LocalPosition() {localOrigin=InstancePreSkinning}
  NamedRerouteDeclaration(None=ShaderStageSwitch(PixelShader=VertexInterpolator(VS=@n73.XYZ).PS, VertexShader=@n73.XYZ)) {name=PS-LocalPos, nodeColor=<rgba 0.5234 0 1 1>}
  NamedRerouteDeclaration(None=@n8.B) {name=DripsAnim_Top, nodeColor=<rgba 0 0.4609 1 1>}
  NamedRerouteDeclaration(None=@n71.R) {name=RandomX, nodeColor=<rgba 1 0.0311 0 1>}
  NamedRerouteDeclaration(None=@n71.G) {name=RandomY, nodeColor=<rgba 0.0142 1 0.0081 1>}
  NamedRerouteDeclaration(None=@n71.B) {name=RandomZ, nodeColor=<rgba 0 0.1197 1 1>}
  NamedRerouteDeclaration(None=@n71.A) {name=RandomA, nodeColor=<rgba 0 1 0.4375 1>}
  NamedRerouteDeclaration(None=@n27) {name=debug, nodeColor=<rgba 0 0.8125 1 1>}
  NamedRerouteUsage()
  ScalarParameter() {defaultValue=2, parameterName=DebugFrame}
 ]
}
