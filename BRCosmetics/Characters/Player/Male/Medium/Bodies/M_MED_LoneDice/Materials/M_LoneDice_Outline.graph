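# Masked, unlit, two-sided inverted-hull outline material for the LoneDice character, showing only back faces in a flat outline color.
# World position offset logic pushes vertices along normals using mask-texture thickness channels, body-bag variants, distance-varying width and noise irregularity, plus a camera offset.
# Attributes pass through MF_CharacterParent_2023_PostFX.
# Overspray, dot and dithered line-mask branches are built but do not reach emissive or opacity mask.
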Material /BRCosmetics/Characters/Player/Male/Medium/Bodies/M_MED_LoneDice/Materials/M_LoneDice_Outline
graph {
 properties {
  bAutomaticallySetUsageInEditor=false
  bUseMaterialAttributes=true
  bUsedWithMeshDeformer=true
  bUsedWithMorphTargets=true
  bUsedWithSkeletalMesh=true
  bUsedWithStaticLighting=true
  blendMode=BLEND_Masked
  shadingModel=MSM_Unlit
  twoSided=true
  usedShadingModels=Unlit
 }
 nodes [
  Add(A=NamedRerouteUsage())
  Composite() {subgraphName=Construct Base Inputs}
  Constant()
  Constant() {r=1}
  n1 = TextureSampleParameter2D() {
   group=01_InvertedHull_Outlines
   parameterName=FX
   samplerType=SAMPLERTYPE_Masks
   sortPriority=1
   texture=/Game/Packages/Fortress_SharedMaterials/Anime_Master_Material/T_Outline_Mask
  }
  n2 = ChannelMaskParameter(None=@n1.RGB) {group=01_InvertedHull_Outlines, parameterName=OutlineThickness_Channel, sortPriority=3}
  n3 = ChannelMaskParameter(None=@n1.RGB) {
   group=01_InvertedHull_Outlines
   maskChannel=Green
   parameterName=OutlineThickness_Channel_BodyBag
   sortPriority=3
   defaultValue=<rgba 0 1 0 0>
  }
  n4 = Subtract(A=@n3.R) {constB=0.5}
  n5 = ScalarParameter() {defaultValue=10, parameterName=BodyBagOutlineThicknessMulti}
  n6 = NamedRerouteUsage()
  n7 = Add(A=LinearInterpolate(A=@n2.R, B=Multiply(A=@n4, B=@n5), Alpha=@n6), B=Constant() {desc=one bit used for opacity cutout, r=-0.009})
  n8 = NamedRerouteUsage()
  LinearInterpolate(
   A=@n7
   Alpha=SmoothStep(Min=Constant() {r=140}, Max=Constant() {r=143.4364}, Value=ComponentMask(None=@n8) {b=true}) {constMax=160, constMin=140}
  )
  n9 = Saturate(None=DotProduct(A=VertexNormalWS(), B=CameraVectorWS()))
  n10 = ScalarParameter() {defaultValue=0.4, group=01_InvertedHull_Outlines, parameterName=LineMask_Power, sliderMax=4}
  n11 = TextureSample(UVs=NamedRerouteUsage()) {
   samplerSource=SSM_Wrap_WorldGroupSettings
   texture=/BRCosmetics/Characters/Player/Male/Medium/Bodies/M_MED_LoneDice/Textures/T_LoneDice_Outline_Noise
  }
  n12 = ConstantBiasScale(None=ComponentMask(None=@n11.RGB) {g=true, r=true}) {bias=-0.5, scale=2}
  n13 = ScalarParameter() {defaultValue=0.2, group=01_InvertedHull_Outlines, parameterName=Distortion_Amount, sliderMax=2}
  n14 = TextureSample(UVs=Add(A=Power(Base=@n9, Exp=@n10), B=Multiply(A=@n12, B=@n13))) {texture=/Game/Global/Textures/Default/Masks_and_Gradients/gradient_stasis}
  LinearInterpolate(
   A=LinearInterpolate(
    A=ScalarParameter() {defaultValue=-6, group=01_InvertedHull_Outlines, parameterName=LineMask_Near, sliderMax=-10}
    B=ScalarParameter() {defaultValue=-1, group=01_InvertedHull_Outlines, parameterName=LineMask_Far, sliderMax=-10}
    Alpha=NamedRerouteUsage()
   )
   B=ScalarParameter() {defaultValue=1.5, group=01_InvertedHull_Outlines, parameterName=LineMask_B, sliderMax=3}
   Alpha=@n14.R
  ) {constA=-2, constB=2}
  n15 = ScalarParameter() {defaultValue=0.5, desc=0.4, group=01_InvertedHull_Outlines, parameterName=LineThickness, sliderMax=2, sortPriority=3}
  n16 = LinearInterpolate(
   A=@n15
   B=ScalarParameter() {defaultValue=0.8, group=01_InvertedHull_Outlines, parameterName=LineThickness_Far, sliderMax=2, sortPriority=3}
   Alpha=NamedRerouteUsage()
  )
  n17 = VectorParameter() {group=LoneDice Outline, parameterName=BodyBagHeadLocation, defaultValue=<rgba 0 0 160 1>}
  n18 = NamedRerouteUsage()
  n19 = ScalarParameter() {defaultValue=30, group=LoneDice Outline, parameterName=LD_BodyBagHeadRadius}
  LinearInterpolate(
   A=@n16
   B=Multiply(A=@n16, B=ScalarParameter() {defaultValue=1, group=LoneDice Outline, parameterName=LD_BodyBagHeadLineThicknessMulti})
   Alpha=NamedRerouteDeclaration(None=SphereMask(A=@n17.RGB, B=@n18, Radius=@n19) {hardnessPercent=0}) {name=BodyBagHead_Mask, nodeColor=<rgba 0.5469 0 1 1>}
  )
  n20 = TwoSidedSign()
  n21 = Constant()
  MaterialExpressionReroute_4 = Reroute(None=@n21)
  n22 = PreSkinnedPosition()
  n23 = ShaderStageSwitch(PixelShader=VertexInterpolator(VS=@n22).PS, VertexShader=@n22)
  MaterialExpressionReroute_5 = Reroute(None=@n23)
  n24 = MaterialFunctionCall!(Material) {materialFunction=/Game/Packages/Fortress_SharedMaterials/Base_Material_Functions/CharacterBrightnessandScale_Balance}
  MaterialExpressionReroute_6 = Reroute(None=@n24.Distance Blend)
  n25 = Fresnel() {baseReflectFraction=0, exponent=1}
  MaterialExpressionReroute_7 = Reroute(None=@n25)
  MaterialFunctionCall(
   Material=SetMaterialAttributes!(
    MaterialAttributes
    Emissive Color=VectorParameter() {group=01_InvertedHull_Outlines, parameterName=Outline_Color, defaultValue=<rgba 0 0 0 1>}.RGB
    Opacity Mask=OneMinus(None=@n20)
    World Position Offset=NamedRerouteUsage()
   ) {attributeSetTypes [#1, #2, #3]}
   M=@MaterialExpressionReroute_4.M
   Pre Skinned Local Position=@MaterialExpressionReroute_5.Pre Skinned Local Position
   Distance Blend=@MaterialExpressionReroute_6.Distance Blend
   Fresnel=@MaterialExpressionReroute_7.Fresnel
  ) {materialFunction=&MF_CharacterParent_2023_PostFX}
  n26 = Constant() {desc=Needs at least a small amount of this or it will break, r=0.25}
  n27 = MaterialFunctionCall() {materialFunction=&CameraDirectionVector}
  n28 = Normalize(VectorInput=Multiply(A=Constant3Vector() {constant=<rgba 1 1 0 0>}, B=@n27.Result))
  n29 = ScalarParameter() {defaultValue=70, group=01_InvertedHull_Outlines, parameterName=VP_CameraVectorMin, sliderMax=100}
  MaterialFunctionCall(
   Plane Normal Axis=Add(A=Multiply(A=@n26, B=@n28), B=@n27.Result)
   Plane Center=Add(A=Multiply(A=@n27.Result, B=@n29), B=ObjectPositionWS())
   UV Scale=ComponentMask(None=VectorParameter() {parameterName=VP_Scale, defaultValue=<rgba 75 75 0 1>}.RGB) {g=true, r=true}
  ) {materialFunction=&VirtualPlaneCoordinates}
  NamedRerouteDeclaration(None=MaterialFunctionCall!(
    Fade Offset=ScalarParameter() {defaultValue=50, group=01_InvertedHull_Outlines, parameterName=Fade_Offset, sliderMax=100}
    Fade Distance=ScalarParameter() {defaultValue=1000, group=01_InvertedHull_Outlines, parameterName=Fade_Distance, sliderMax=1500, sliderMin=500}
   ) {materialFunction=/Game/Global/Materials/Character_Parent_2023/Anime_Parent_2023/MaterialFunctions/MFC_DistanceFade_FOV_compensation}.Result) {name=DistanceFading, nodeColor=<rgba 1 0.3047 0 1>}
  NamedRerouteDeclaration(None=MaterialFunctionCall!(Position, UVRadialScale=ScalarParameter() {defaultValue=0.4, group=LoneDice Outline, parameterName=LD_OutlineNoiseUVScale}) {materialFunction=/Game/Characters/Player/Common/MaterialFunctions/Superlevel/CharacterStabilizedScreenSpaceUVs}.StabilizedUVs) {name=UVs, nodeColor=<rgba 1 0.9141 0 1>}
  n30 = LocalPosition() {localOrigin=InstancePreSkinning}
  NamedRerouteDeclaration(None=ShaderStageSwitch(PixelShader=VertexInterpolator(VS=@n30.XYZ).PS, VertexShader=@n30.XYZ)) {name=PS-LocalPos, nodeColor=<rgba 1 0 0.2578 1>}
  n31 = Step(X=ComponentMask(None=TextureCoordinate()) {r=true}) {constY=1}
  NamedRerouteDeclaration(None=@n31) {name=Debug, nodeColor=<rgba 0.6406 1 0 1>}
  NamedRerouteDeclaration(None=VertexInterpolator(VS=PreSkinnedNormal()).PS) {name=PS-LocalNormal, nodeColor=<rgba 1 0 0.0234 1>}
  NamedRerouteDeclaration(None=@n11.B) {name=Dots, nodeColor=<rgba 0 0.8125 1 1>}
  n32 = NamedRerouteUsage()
  n33 = Divide(A=Add(A=VectorParameter() {parameterName=Offset, defaultValue=<rgba -6.7699 0 1 1>}.RGB, B=@n32)) {constB=20}
  n34 = TextureSample(UVs=Rotator(Coordinate=ComponentMask(None=@n33) {b=true, r=true}, Time=Constant() {r=2})) {
   samplerSource=SSM_Wrap_WorldGroupSettings
   texture=&T_LoneDice_Outline_Noise
  }
  n35 = ComponentMask(None=Divide(A=@n32) {constB=64}) {b=true, g=true}
  n36 = TextureSample(UVs=@n35) {
   samplerSource=SSM_Wrap_WorldGroupSettings
   texture=&T_LoneDice_Outline_Noise
  }
  n37 = Multiply(A=NamedRerouteDeclaration(None=@n36.R) {name=OutlineIrregularity, nodeColor=<rgba 0 0.6484 1 1>}, B=@n36.G)
  n38 = Abs(None=@n35)
  n39 = Multiply(
   A=Multiply(A=@n34.A, B=SmoothStep(Value=@n37) {constMax=0.3})
   B=SmoothStep(Value=ComponentMask(None=@n38) {r=true}) {constMax=0.15, desc="Mask out  stretched lines"}
  )
  n40 = Saturate(None=Add(A=NamedRerouteUsage()) {constB=0.3})
  NamedRerouteDeclaration(None=LinearInterpolate(B=Multiply(A=@n39, B=@n40), Alpha=ScalarParameter() {defaultValue=1, group=LoneDice Outline, parameterName=OversprayOpacity, sliderMax=1})) {name=Overspray, nodeColor=<rgba 0 1 0.0391 1>}
  n41 = SmoothStep(Value=Fresnel(ExponentIn=ScalarParameter() {defaultValue=1.5, group=LoneDice Outline, parameterName=LD_OutlineFresnelExp}) {baseReflectFraction=0, exponent=1}) {constMin=0.5}
  n42 = Multiply(
   A=@n41
   B=MaterialFunctionCall(Alpha Threshold=@n41) {materialFunction=&DitherTemporalAA}.Result
  )
  NamedRerouteDeclaration(None=Multiply(A=Saturate(None=@n42), B=ScalarParameter() {group=LoneDice Outline, parameterName=LD_OutlineFresnel})) {name=Opacity_OUT, nodeColor=<rgba 0 1 0.5313 1>}
  NamedRerouteDeclaration(None=@n31) {name=BodyBagMask, nodeColor=<rgba 0.4062 1 0 1>}
  NamedRerouteDeclaration(None=SmoothStep(Value=Fresnel() {baseReflectFraction=0, exponent=1}) {constMax=0.6}) {name=Fresnel, nodeColor=<rgba 1 0.8906 0 1>}
  n43 = VertexNormalWS()
  n44 = Multiply(
   A=Subtract(A=NamedRerouteUsage()) {constB=0.5}
   B=ScalarParameter() {defaultValue=3, group=LoneDice Outline, parameterName=LD_OutlineIrregularityScale}
  )
  n45 = LinearInterpolate(A=Multiply(A=@n44, B=Constant() {r=0.5}), B=@n44, Alpha=LinearInterpolate(A=@n2.R, B=@n3.R, Alpha=@n6))
  n46 = Multiply(
   A=StaticSwitchParameter(True=Add(A=@n45, B=@n7) {constB=0}, False=Constant() {r=1}) {group=01_InvertedHull_Outlines, parameterName=UseThicknessMap, sortPriority=0}
   B=@n16
  )
  n47 = LinearInterpolate(
   A=ScalarParameter() {group=LoneDice Outline, parameterName=LD_CamOffset}
   B=ScalarParameter() {group=LoneDice Outline, parameterName=LD_CamOffset_BodyBagHead}
   Alpha=NamedRerouteUsage()
  )
  NamedRerouteDeclaration(None=Add(
    A=Multiply(A=@n43, B=@n46)
    B=MaterialFunctionCall(Offset Amount=@n47) {materialFunction=/Engine/Functions/Engine_MaterialFunctions02/WorldPositionOffset/CameraOffset}.Camera Offset
   )) {name=WPO_OUT, nodeColor=<rgba 1 0.2578 0 1>}
  n48 = ComponentMask(None=MaterialFunctionCall!(
    ColorTex=TextureObject() {texture=&T_LoneDice_Outline_Noise}
    NormalTex
    Pos=NamedRerouteUsage()
    Nrm
    ColorUVScale=Constant() {r=70}
    ColorBlendFactor
    NormalStrength
    NormalUVScale
    NormalBlendFactor
    Local Space Inputs
   ) {materialFunction=/Game/Global/MaterialFunctions/MF_BiPlanar}.Color) {b=true}
  n49 = NamedRerouteUsage()
  NamedRerouteDeclaration(None=Multiply(A=Multiply(A=@n48, B=@n49), B=ScalarParameter() {defaultValue=1, group=LoneDice Outline, parameterName=DotsIntensity})) {name=BigDots, nodeColor=<rgba 0 1 0.0625 1>}
  NamedRerouteUsage()
  NamedRerouteUsage()
  PinBase!(M=@n21, Pre Skinned Local Position=@n23, Distance Blend=@n24.Distance Blend, Fresnel=@n25) {reroutePins [
    M {expression=&self:MaterialExpressionReroute_4}
    Pre Skinned Local Position {expression=&self:MaterialExpressionReroute_5}
    Distance Blend {expression=&self:MaterialExpressionReroute_6}
    Fresnel {expression=&self:MaterialExpressionReroute_7}
   ]}
  PinBase!()
  n50 = OneMinus(None=Multiply(A=ScalarParameter() {defaultValue=1.2, group=01_InvertedHull_Outlines, parameterName=LineMask_InnerIntensity, sliderMax=2}, B=@n9))
  n51 = Saturate(None=@n14.R)
  n52 = Multiply(
   A=@n51
   B=MaterialFunctionCall(Alpha Threshold=@n51) {materialFunction=&DitherTemporalAA}.Result
  )
  Saturate(None=Multiply(
    A=ScalarParameter() {defaultValue=1, group=01_InvertedHull_Outlines, parameterName=Outline_Opacity, sliderMax=2}
    B=Multiply(A=@n50, B=@n52) {desc=Mask out inner}
   ))
  n53 = NamedRerouteUsage()
  Saturate(None=Multiply(A=Multiply(A=@n53, B=@n50)) {constB=2})
 ]
}
