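# Master material for the LoneDice male player character, layering custom LoneDice shading into the shared 2023 character parent pipeline.
# A static switch, UseEyes, picks the 2023 eye function or the parent PreFX function as base material attributes.
# MF_LoneDice_Shading modifies those attributes, then the parent PostFX function finishes them, with auxiliary inputs wired from PreFX.
# Shading model comes from the expression, with a hero skin subsurface profile assigned.
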Material /BRCosmetics/Characters/Player/Male/Medium/Bodies/M_MED_LoneDice/Materials/M_LoneDice_Shading
graph {
 properties {
  bAutomaticallySetUsageInEditor=false
  bUseMaterialAttributes=true
  bUsedWithMeshDeformer=true
  bUsedWithMorphTargets=true
  bUsedWithSkeletalMesh=true
  pixelDepthOffsetMode=PDOM_Legacy
  shadingModel=MSM_FromMaterialExpression
  subsurfaceProfile=&SS_HeroSkin_02
  usedShadingModels=Default Lit | Subsurface | Clear Coat | Subsurface Profile | Cloth
 }
 nodes [
  n1 = MaterialFunctionCall() {materialFunction=/Game/Global/Materials/Character_Parent_2023/Eye_Parent_2023/MF_eyes_2023}
  n2 = MaterialFunctionCall() {materialFunction=&MF_CharacterParent_2023_PreFX}
  MaterialFunctionCall(
   Material=MaterialFunctionCall!(MA_In=StaticSwitchParameter(True=@n1.Result, False=@n2.Result) {group=LoneDice, parameterName=UseEyes}, M=@n2.M) {materialFunction=&/MF_LoneDice_Shading}.MA_Out
   M=@n2.M
   Pre Skinned Local Position=@n2.M
   Distance Blend=@n2.M
   Fresnel=@n2.M
  ) {materialFunction=&MF_CharacterParent_2023_PostFX}
 ]
}
