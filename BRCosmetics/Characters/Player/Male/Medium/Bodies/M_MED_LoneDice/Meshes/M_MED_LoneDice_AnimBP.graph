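# Post-copy AnimBlueprint for the LoneDice medium male body cosmetic, adding secondary motion to the parent mesh's pose.
# Weighted constraints place spray-gun and zipper bones, Modify Bone nodes offset backpack and amulet, and AnimDynamics simulates spray guns, zippers and amulet.
# A final Control Rig receives HangTwoArms, set by parachute and glider events.
# The event graph also computes GravityOverride and a skydiving alpha, neither read by the AnimGraph.
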AnimBlueprint /BRCosmetics/Characters/Player/Male/Medium/Bodies/M_MED_LoneDice/Meshes/M_MED_LoneDice_AnimBP
target_skeleton=&/M_MED_LoneDice_Skeleton
graphs [
 AnimGraph {nodes [
   n1 = "Constraint - base_spraygun (PPPPPPPP)"!(
    ComponentPose=Local To Component(LocalPose=Copy Pose From Mesh() {bUseAttachedParent=true})
    Alpha=1
    ConstraintWeights_0=0.7137
    ConstraintWeights_1:float=0.0931
    ConstraintWeights_2:float=0.1102
    ConstraintWeights_3:float=0.0666
    ConstraintWeights_4:float=0.0025
    ConstraintWeights_5:float=0
    ConstraintWeights_6:float=0.0002
    ConstraintWeights_7:float=0.0137
   ) -> (Pose) {
    alphaBoolBlend {blendOption=Linear}
    boneToModify {boneName=base_spraygun}
    constraintSetup [
     {targetBone {boneName=pelvis}}
     {targetBone {boneName=spine_01}}
     {targetBone {boneName=spine_02}}
     {targetBone {boneName=spine_03}}
     {targetBone {boneName=spine_04}}
     {targetBone {boneName=Correct_belt_m}}
     {targetBone {boneName=Correct_belt_l}}
     {targetBone {boneName=Correct_belt_side_l}}
    ]
    constraintWeights [1,1,1,1,1,1,1,1]
   }
   n2 = "Transform (Modify) Bone - Bone: attach_backpack"(ComponentPose=@n1, Alpha=1, Translation="0,2,0", Rotation="0,0,0", Scale="1,1,1") {
    translationMode=BMM_Additive
    translationSpace=BCS_BoneSpace
    alphaBoolBlend {blendOption=Linear}
    boneToModify {boneName=attach_backpack}
    translation {y=2}
   }
   n3 = Anim Dynamics - Bone: dyn_SprayGun_l(ComponentPose=@n2, Alpha=1, ExternalForce="0,0,0") {
    angularDampingOverride=0.99
    angularSpringConstant=500
    bOverrideAngularDamping=true
    bUseGravityOverride=true
    bUsePlanarLimit=false
    bUseSphericalLimits=true
    gravityScale=-1
    simulationSpace=RootRelative
    alphaBoolBlend {blendOption=Linear}
    boundBone {boneName=dyn_SprayGun_l}
    componentAppliedLinearAccClamp=<5000 5000 5000>
    componentLinearAccScale=<0.2 0.2 0.2>
    componentLinearVelScale=<0.2 0.2 0.2>
    gravityOverride {z=-980}
    physicsBodyDefinitions [{
      collisionType=CustomSphere
      sphereCollisionRadius=1
      boundBone {boneName=dyn_SprayGun_l}
      boxExtents=<10 5 5>
      constraintSetup {angularLimitsMax=<0 0 0>, angularLimitsMin=<0 0 0>, angularTarget=<0 0 0>}
      localJointOffset=<0 0 0>
     }]
   }
   n4 = Anim Dynamics - Bone: dyn_SprayGun_r(ComponentPose=@n3, Alpha=1, ExternalForce="0,0,0") {
    angularDampingOverride=0.99
    angularSpringConstant=500
    bOverrideAngularDamping=true
    bUseGravityOverride=true
    bUsePlanarLimit=false
    bUseSphericalLimits=true
    gravityScale=-1
    simulationSpace=RootRelative
    alphaBoolBlend {blendOption=Linear}
    boundBone {boneName=dyn_SprayGun_r}
    componentAppliedLinearAccClamp=<5000 5000 5000>
    componentLinearAccScale=<0.2 0.2 0.2>
    componentLinearVelScale=<0.2 0.2 0.2>
    gravityOverride {z=-980}
    physicsBodyDefinitions [{
      collisionType=CustomSphere
      sphereCollisionRadius=1
      boundBone {boneName=dyn_SprayGun_r}
      boxExtents=<10 5 5>
      constraintSetup {angularLimitsMax=<0 0 0>, angularLimitsMin=<0 0 0>, angularTarget=<0 0 0>}
      localJointOffset=<0 0 0>
     }]
   }
   n5 = "Constraint - correct_zipper_upper_r (PPPPPPPP)"!(
    ComponentPose=@n4
    Alpha=1
    ConstraintWeights_0=0
    ConstraintWeights_1:float=0.0895
    ConstraintWeights_2:float=0.2534
    ConstraintWeights_3:float=0.6556
    ConstraintWeights_4:float=0.0008
    ConstraintWeights_5:float=0
    ConstraintWeights_6:float=0
    ConstraintWeights_7:float=0.0007
   ) -> (Pose) {
    alphaBoolBlend {blendOption=Linear}
    boneToModify {boneName=correct_zipper_upper_r}
    constraintSetup [
     {targetBone {boneName=thigh_r}}
     {targetBone {boneName=calf_r}}
     {targetBone {boneName=calf_twist_01_r}}
     {targetBone {boneName=calf_twist_02_r}}
     {targetBone {boneName=foot_r}}
     {targetBone {boneName=deform_knee_out_r}}
     {targetBone {boneName=deform_knee_in_r}}
     {targetBone {boneName=Correct_boot_back_R}}
    ]
    constraintWeights [1,1,1,1,1,1,1,1]
   }
   n6 = "Constraint - correct_zipper_lower_r (PPPPPPP)"!(
    ComponentPose=@n5
    Alpha=1
    ConstraintWeights_0=0
    ConstraintWeights_1:float=0.0201
    ConstraintWeights_2:float=0.3822
    ConstraintWeights_3:float=0.29
    ConstraintWeights_4:float=0.1346
    ConstraintWeights_5:float=0
    ConstraintWeights_6:float=0.173
   ) -> (Pose) {
    alphaBoolBlend {blendOption=Linear}
    boneToModify {boneName=correct_zipper_lower_r}
    constraintSetup [
     {targetBone {boneName=thigh_r}}
     {targetBone {boneName=calf_r}}
     {targetBone {boneName=calf_twist_01_r}}
     {targetBone {boneName=calf_twist_02_r}}
     {targetBone {boneName=foot_r}}
     {targetBone {boneName=deform_knee_in_r}}
     {targetBone {boneName=Correct_boot_back_R}}
    ]
    constraintWeights [1,1,1,1,1,1,1]
   }
   n7 = "Constraint - dyn_zipper_lower_l (PPPPPPP)"!(
    ComponentPose=@n6
    Alpha=1
    ConstraintWeights_0=0
    ConstraintWeights_1:float=0.0212
    ConstraintWeights_2:float=0.3989
    ConstraintWeights_3:float=0.3044
    ConstraintWeights_4:float=0.1219
    ConstraintWeights_5:float=0
    ConstraintWeights_6:float=0.1536
   ) -> (Pose) {
    alphaBoolBlend {blendOption=Linear}
    boneToModify {boneName=dyn_zipper_lower_l}
    constraintSetup [
     {targetBone {boneName=thigh_l}}
     {targetBone {boneName=calf_l}}
     {targetBone {boneName=calf_twist_01_l}}
     {targetBone {boneName=calf_twist_02_l}}
     {targetBone {boneName=foot_l}}
     {targetBone {boneName=deform_knee_in_l}}
     {targetBone {boneName=Correct_boot_back_L}}
    ]
    constraintWeights [1,1,1,1,1,1,1]
   }
   n8 = "Constraint - correct_zipper_upper_l (PPPPPPPP)"!(
    ComponentPose=@n7
    Alpha=1
    ConstraintWeights_0=0
    ConstraintWeights_1:float=0.0899
    ConstraintWeights_2:float=0.2515
    ConstraintWeights_3:float=0.6567
    ConstraintWeights_4:float=0.0009
    ConstraintWeights_5:float=0
    ConstraintWeights_6:float=0
    ConstraintWeights_7:float=0.0009
   ) -> (Pose) {
    alphaBoolBlend {blendOption=Linear}
    boneToModify {boneName=correct_zipper_upper_l}
    constraintSetup [
     {targetBone {boneName=thigh_l}}
     {targetBone {boneName=calf_l}}
     {targetBone {boneName=calf_twist_01_l}}
     {targetBone {boneName=calf_twist_02_l}}
     {targetBone {boneName=foot_l}}
     {targetBone {boneName=deform_knee_out_l}}
     {targetBone {boneName=deform_knee_in_l}}
     {targetBone {boneName=Correct_boot_back_L}}
    ]
    constraintWeights [1,1,1,1,1,1,1,1]
   }
   n9 = Anim Dynamics - Bone: dyn_zipper_upper_l(ComponentPose=@n8, Alpha=1, ExternalForce="0,0,0") {
    angularDampingOverride=0.99
    angularSpringConstant=500
    bOverrideAngularDamping=true
    bUseGravityOverride=true
    bUsePlanarLimit=false
    bUseSphericalLimits=true
    gravityScale=-1
    simulationSpace=RootRelative
    alphaBoolBlend {blendOption=Linear}
    boundBone {boneName=dyn_zipper_upper_l}
    componentAppliedLinearAccClamp=<5000 5000 5000>
    componentLinearAccScale=<0.2 0.2 0.2>
    componentLinearVelScale=<0.2 0.2 0.2>
    gravityOverride {z=-980}
    physicsBodyDefinitions [{
      collisionType=CustomSphere
      sphereCollisionRadius=1
      boundBone {boneName=dyn_zipper_upper_l}
      boxExtents=<10 5 5>
      constraintSetup {angularLimitsMax=<5 10 60>, angularLimitsMin=<-5 -10 -5>, angularTarget=<0 0 0>}
      localJointOffset=<1 0 0>
     }]
   }
   n10 = Anim Dynamics - Bone: dyn_zipper_lower_l(ComponentPose=@n9, Alpha=1, ExternalForce="0,0,0") {
    angularDampingOverride=0.99
    angularSpringConstant=500
    bOverrideAngularDamping=true
    bUseGravityOverride=true
    bUsePlanarLimit=false
    bUseSphericalLimits=true
    gravityScale=-1
    simulationSpace=RootRelative
    alphaBoolBlend {blendOption=Linear}
    boundBone {boneName=dyn_zipper_lower_l}
    componentAppliedLinearAccClamp=<5000 5000 5000>
    componentLinearAccScale=<0.2 0.2 0.2>
    componentLinearVelScale=<0.2 0.2 0.2>
    gravityOverride {z=-980}
    physicsBodyDefinitions [{
      collisionType=CustomSphere
      sphereCollisionRadius=1
      boundBone {boneName=dyn_zipper_lower_l}
      boxExtents=<10 5 5>
      constraintSetup {angularLimitsMax=<5 10 60>, angularLimitsMin=<-5 -10 -5>, angularTarget=<0 0 0>}
      localJointOffset=<1 0 0>
     }]
   }
   n11 = Anim Dynamics - Bone: dyn_zipper_upper_r(ComponentPose=@n10, Alpha=1, ExternalForce="0,0,0") {
    angularDampingOverride=0.99
    angularSpringConstant=500
    bOverrideAngularDamping=true
    bUseGravityOverride=true
    bUsePlanarLimit=false
    bUseSphericalLimits=true
    gravityScale=-1
    simulationSpace=RootRelative
    alphaBoolBlend {blendOption=Linear}
    boundBone {boneName=dyn_zipper_upper_r}
    componentAppliedLinearAccClamp=<5000 5000 5000>
    componentLinearAccScale=<0.2 0.2 0.2>
    componentLinearVelScale=<0.2 0.2 0.2>
    gravityOverride {z=-980}
    physicsBodyDefinitions [{
      collisionType=CustomSphere
      sphereCollisionRadius=1
      boundBone {boneName=dyn_zipper_upper_r}
      boxExtents=<10 5 5>
      constraintSetup {angularLimitsMax=<5 10 60>, angularLimitsMin=<-5 -10 -5>, angularTarget=<0 0 0>}
      localJointOffset=<1 0 0>
     }]
   }
   n12 = Anim Dynamics - Bone: dyn_zipper_lower_r(ComponentPose=@n11, Alpha=1, ExternalForce="0,0,0") {
    angularDampingOverride=0.99
    angularSpringConstant=500
    bOverrideAngularDamping=true
    bUseGravityOverride=true
    bUsePlanarLimit=false
    bUseSphericalLimits=true
    gravityScale=-1
    simulationSpace=RootRelative
    alphaBoolBlend {blendOption=Linear}
    boundBone {boneName=dyn_zipper_lower_r}
    componentAppliedLinearAccClamp=<5000 5000 5000>
    componentLinearAccScale=<0.2 0.2 0.2>
    componentLinearVelScale=<0.2 0.2 0.2>
    gravityOverride {z=-980}
    physicsBodyDefinitions [{
      collisionType=CustomSphere
      sphereCollisionRadius=1
      boundBone {boneName=dyn_zipper_lower_r}
      boxExtents=<10 5 5>
      constraintSetup {angularLimitsMax=<5 10 60>, angularLimitsMin=<-5 -10 -5>, angularTarget=<0 0 0>}
      localJointOffset=<1 0 0>
     }]
   }
   n13 = "Transform (Modify) Bone - Bone: Dyn_amulet_01"(ComponentPose=@n12, Alpha=1, Translation="0,0,0", Rotation="2,-3,0", Scale="1,0.95,1") {
    rotationMode=BMM_Additive
    rotationSpace=BCS_BoneSpace
    scaleMode=BMM_Additive
    scaleSpace=BCS_BoneSpace
    alphaBoolBlend {blendOption=Linear}
    boneToModify {boneName=Dyn_amulet_01}
    rotation {pitch=2, yaw=-3}
    scale {y=0.95}
   }
   n14 = "Transform (Modify) Bone - Bone: Dyn_amulet_02"(ComponentPose=@n13, Alpha=1, Translation="0,0,0", Rotation="-3.568,0,0", Scale="1,0.9,1") {
    rotationMode=BMM_Additive
    rotationSpace=BCS_BoneSpace
    scaleMode=BMM_Additive
    scaleSpace=BCS_BoneSpace
    alphaBoolBlend {blendOption=Linear}
    boneToModify {boneName=Dyn_amulet_02}
    rotation {pitch=-3.568}
    scale {y=0.9}
   }
   n15 = "Transform (Modify) Bone - Bone: Dyn_amulet_03"(ComponentPose=@n14, Alpha=1, Translation="0,0,0", Rotation="8.237,0,0", Scale="1,0.85,1") {
    rotationMode=BMM_Additive
    rotationSpace=BCS_BoneSpace
    scaleMode=BMM_Additive
    scaleSpace=BCS_BoneSpace
    alphaBoolBlend {blendOption=Linear}
    boneToModify {boneName=Dyn_amulet_03}
    rotation {pitch=8.237}
    scale {y=0.85}
   }
   n16 = Anim Dynamics - Bone: Dyn_amulet_02(ComponentPose=@n15, Alpha=1, ExternalForce="0,0,0") {
    angularDampingOverride=0.999
    angularSpringConstant=500
    bAngularSpring=true
    bGravityOverrideInSimSpace=true
    bOverrideAngularDamping=true
    bUseGravityOverride=true
    bUsePlanarLimit=false
    bUseSphericalLimits=true
    simulationSpace=RootRelative
    alphaBoolBlend {blendOption=Linear}
    boundBone {boneName=Dyn_amulet_02}
    componentAppliedLinearAccClamp=<3500 3500 3500>
    componentLinearAccScale=<1.2 1.2 1.2>
    componentLinearVelScale=<1.2 1.2 1.2>
    gravityOverride {z=-980}
    physicsBodyDefinitions [{
      boundBone {boneName=Dyn_amulet_02}
      boxExtents=<20 13 15>
      constraintSetup {angularLimitsMax=<0 1 1>, angularLimitsMin=<0 -10 -1>, angularTarget=<0 0 0>}
      localJointOffset=<4 0 0>
     }]
    relativeSpaceBone {boneName=root}
    sphericalLimits [
     {limitRadius=10, drivingBone {boneName=spine_05}, sphereLocalOffset=<-1.2187 3.2294 0>}
     {limitRadius=10, drivingBone {boneName=spine_05}, sphereLocalOffset=<-1.2187 3.5902 4.8835>}
    ]
   }
   n17 = Anim Dynamics - Bone: Dyn_amulet_03(ComponentPose=@n16, Alpha=1, ExternalForce="0,0,0") {
    angularDampingOverride=0.999
    angularSpringConstant=500
    bAngularSpring=true
    bOverrideAngularDamping=true
    bUseGravityOverride=true
    bUsePlanarLimit=false
    bUseSphericalLimits=true
    simulationSpace=RootRelative
    alphaBoolBlend {blendOption=Linear}
    boundBone {boneName=Dyn_amulet_03}
    componentAppliedLinearAccClamp=<3500 3500 3500>
    componentLinearAccScale=<1.2 1.2 1.2>
    componentLinearVelScale=<1.2 1.2 1.2>
    gravityOverride {z=-980}
    physicsBodyDefinitions [{
      boundBone {boneName=Dyn_amulet_03}
      boxExtents=<20 20 15>
      constraintSetup {angularLimitsMax=<5 1 1>, angularLimitsMin=<-5 -20 -1>, angularTarget=<0 0 0>}
      localJointOffset=<4 0 0>
     }]
    relativeSpaceBone {boneName=root}
    sphericalLimits [
     {limitRadius=10, drivingBone {boneName=spine_05}, sphereLocalOffset=<-1.2187 3.2294 0>}
     {limitRadius=10, drivingBone {boneName=spine_05}, sphereLocalOffset=<-1.2187 3.5902 4.8835>}
    ]
   }
   n18 = Control Rig!(Source=Component To Local(ComponentPose=@n17), Alpha, HangTwoArms:double=Get Hang Two Arms() -> (Hang Two Arms:double)) -> (Pose) {
    alphaBoolBlend {blendOption=Linear}
    controlRigAssetReference {blueprintRigClass=&/M_MED_LoneDice_CtrlRig.M_MED_LoneDice_CtrlRig_C}
   }
   Output Pose(Result=@n18) {layerGroup=None}
  ]}
 EventGraph {nodes [
   Event BlueprintUpdateAnimation()
   > n1 = Sequence()
   > n2 = Gravity Override State Function DEPRECATED(
    CharacterPartAnimInstance=Self-Reference()
    GameTimeInSeconds=GetGameTimeInSeconds()
    GravityOverrideParams=Get Gravity Override Params()
    DebugJointName=~
   )
   > Set Gravity Override(Gravity Override=@n2)
   n3 = Get bIsParachuteOpened()
   n4 = Branch(execute=@n1.then_1, Condition=AND Boolean(A=Get bIsSkyDiving(), B=NOT Boolean(A=@n3)))
   > Set Sky Aplha(Sky Aplha:double=1) -> (Output_Get:double)
   Set Sky Aplha(execute=[@n5.then, @n4.else], Sky Aplha=0)
   n5 = Event OnEndSkydiving()
   Event OnParachuteClosed()
   > Set Hang Two Arms(Hang Two Arms:double=0) -> (Output_Get:double)
   Event OnParachuteOpened()
   > Switch on EGliderType(Selection=Get PlayerGliderType() -> (PlayerGliderType=HangGlider))
   > Set Hang Two Arms(Hang Two Arms=1)
  ]}
]
variables [
 Gravity Override {type=Vector}
 Gravity Override Params {type=GravityOverrideParamsStruct}
 Hang Two Arms {type=number, default=1}
 Sky Aplha {type=number, default=1}
]
