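# Anim Blueprint for the LoneDice male face accessory mesh that adds corrective offsets and secondary motion to its hair bones.
# It copies the attached parent's pose, then applies stacked additive Transform Bone adjustments to the back and scale hair corrective bones.
# Seven chained Anim Dynamics nodes simulate hair_01 to hair_07 root-relative, with damped, angle-limited motion.
# The event graph sets LOD 1 and LOD 2+ flags, which the AnimGraph does not visibly read.
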AnimBlueprint /BRCosmetics/Characters/Player/Male/Medium/Bodies/M_MED_LoneDice/Meshes/Parts/M_MED_LoneDice_FaceAcc_AnimBP
target_skeleton=&/M_MED_LoneDice_FaceAcc_Skeleton
graphs [
 AnimGraph {nodes [
   n1 = "Transform (Modify) Bone - Bone: Hair_Corrective_Back"!(
    ComponentPose=Local To Component(LocalPose=Copy Pose From Mesh() {bUseAttachedParent=true, rootBoneToCopy=root})
    bAlphaBoolEnabled:bool=True
    Translation="0,-0.1,0"
    Rotation="0,0,0"
    Scale="1,1,1"
   ) -> (Pose) {
    alphaInputType=Bool
    scaleSpace=BCS_BoneSpace
    translationMode=BMM_Additive
    translationSpace=BCS_BoneSpace
    alphaBoolBlend {blendOption=Linear}
    boneToModify {boneName=Hair_Corrective_Back}
    translation {y=-0.1}
   }
   n2 = "Transform (Modify) Bone - Bone: Hair_Corrective_scale"!(ComponentPose=@n1, bAlphaBoolEnabled:bool=True, Translation="0,0,0", Rotation="0,0,0", Scale="1.03,1,1") -> (Pose) {
    alphaInputType=Bool
    scaleMode=BMM_Additive
    scaleSpace=BCS_BoneSpace
    translationSpace=BCS_BoneSpace
    alphaBoolBlend {blendOption=Linear}
    boneToModify {boneName=Hair_Corrective_scale}
    scale {x=1.03}
   }
   n3 = "Transform (Modify) Bone - Bone: Hair_Corrective_Back"(ComponentPose=@n2, bAlphaBoolEnabled=True, Translation="0,-0.6,0", Rotation="0,0,-5", Scale="1,1,1") {
    alphaInputType=Bool
    rotationMode=BMM_Additive
    rotationSpace=BCS_BoneSpace
    translationMode=BMM_Additive
    translationSpace=BCS_BoneSpace
    alphaBoolBlend {blendOption=Linear}
    boneToModify {boneName=Hair_Corrective_Back}
    rotation {roll=-5}
    translation {y=-0.6}
   }
   n4 = "Transform (Modify) Bone - Bone: Hair_Corrective_scale"(ComponentPose=@n3, bAlphaBoolEnabled=True, Translation="0,0,0", Rotation="0,0,0", Scale="1.1,1,1") {
    alphaInputType=Bool
    scaleMode=BMM_Additive
    scaleSpace=BCS_BoneSpace
    translationMode=BMM_Additive
    translationSpace=BCS_BoneSpace
    alphaBoolBlend {blendOption=Linear}
    boneToModify {boneName=Hair_Corrective_scale}
    scale {x=1.1}
   }
   n5 = Anim Dynamics - Bone: hair_01(ComponentPose=@n4, Alpha=1, ExternalForce="0,0,0") {
    angularDampingOverride=0.99
    angularSpringConstant=500
    bOverrideAngularDamping=true
    bUseGravityOverride=true
    bUsePlanarLimit=false
    bUseSphericalLimits=true
    gravityScale=-1
    simulationSpace=RootRelative
    alphaBoolBlend {blendOption=Linear}
    boundBone {boneName=hair_01}
    componentAppliedLinearAccClamp=<5000 5000 5000>
    componentLinearAccScale=<0.2 0.2 0.2>
    componentLinearVelScale=<0.2 0.2 0.2>
    gravityOverride {z=-980}
    physicsBodyDefinitions [{
      collisionType=CustomSphere
      sphereCollisionRadius=1
      boundBone {boneName=hair_01}
      boxExtents=<10 5 5>
      constraintSetup {angularLimitsMax=<5 10 60>, angularLimitsMin=<-5 -10 -5>, angularTarget=<0 0 0>}
      localJointOffset=<0 0 1>
     }]
   }
   n6 = Anim Dynamics - Bone: hair_02(ComponentPose=@n5, Alpha=1, ExternalForce="0,0,0") {
    angularDampingOverride=0.99
    angularSpringConstant=500
    bOverrideAngularDamping=true
    bUseGravityOverride=true
    bUsePlanarLimit=false
    bUseSphericalLimits=true
    gravityScale=-1
    simulationSpace=RootRelative
    alphaBoolBlend {blendOption=Linear}
    boundBone {boneName=hair_02}
    componentAppliedLinearAccClamp=<5000 5000 5000>
    componentLinearAccScale=<0.2 0.2 0.2>
    componentLinearVelScale=<0.2 0.2 0.2>
    gravityOverride {z=-980}
    physicsBodyDefinitions [{
      collisionType=CustomSphere
      sphereCollisionRadius=1
      boundBone {boneName=hair_02}
      boxExtents=<10 5 5>
      constraintSetup {angularLimitsMax=<5 10 60>, angularLimitsMin=<-5 -10 -5>, angularTarget=<0 0 0>}
      localJointOffset=<0 0 1>
     }]
   }
   n7 = Anim Dynamics - Bone: hair_03(ComponentPose=@n6, Alpha=1, ExternalForce="0,0,0") {
    angularDampingOverride=0.99
    angularSpringConstant=500
    bOverrideAngularDamping=true
    bUseGravityOverride=true
    bUsePlanarLimit=false
    bUseSphericalLimits=true
    gravityScale=-1
    simulationSpace=RootRelative
    alphaBoolBlend {blendOption=Linear}
    boundBone {boneName=hair_03}
    componentAppliedLinearAccClamp=<5000 5000 5000>
    componentLinearAccScale=<0.2 0.2 0.2>
    componentLinearVelScale=<0.2 0.2 0.2>
    gravityOverride {z=-980}
    physicsBodyDefinitions [{
      collisionType=CustomSphere
      sphereCollisionRadius=1
      boundBone {boneName=hair_03}
      boxExtents=<10 5 5>
      constraintSetup {angularLimitsMax=<5 10 60>, angularLimitsMin=<-5 -10 -5>, angularTarget=<0 0 0>}
      localJointOffset=<0 0 1>
     }]
   }
   n8 = Anim Dynamics - Bone: hair_04(ComponentPose=@n7, Alpha=1, ExternalForce="0,0,0") {
    angularDampingOverride=0.99
    angularSpringConstant=500
    bOverrideAngularDamping=true
    bUseGravityOverride=true
    bUsePlanarLimit=false
    bUseSphericalLimits=true
    gravityScale=-1
    simulationSpace=RootRelative
    alphaBoolBlend {blendOption=Linear}
    boundBone {boneName=hair_04}
    componentAppliedLinearAccClamp=<5000 5000 5000>
    componentLinearAccScale=<0.2 0.2 0.2>
    componentLinearVelScale=<0.2 0.2 0.2>
    gravityOverride {z=-980}
    physicsBodyDefinitions [{
      collisionType=CustomSphere
      sphereCollisionRadius=1
      boundBone {boneName=hair_04}
      boxExtents=<10 5 5>
      constraintSetup {angularLimitsMax=<5 10 60>, angularLimitsMin=<-5 -10 -5>, angularTarget=<0 0 0>}
      localJointOffset=<0 0 1>
     }]
   }
   n9 = Anim Dynamics - Bone: hair_05(ComponentPose=@n8, Alpha=1, ExternalForce="0,0,0") {
    angularDampingOverride=0.99
    angularSpringConstant=500
    bOverrideAngularDamping=true
    bUseGravityOverride=true
    bUsePlanarLimit=false
    bUseSphericalLimits=true
    gravityScale=-1
    simulationSpace=RootRelative
    alphaBoolBlend {blendOption=Linear}
    boundBone {boneName=hair_05}
    componentAppliedLinearAccClamp=<5000 5000 5000>
    componentLinearAccScale=<0.2 0.2 0.2>
    componentLinearVelScale=<0.2 0.2 0.2>
    gravityOverride {z=-980}
    physicsBodyDefinitions [{
      collisionType=CustomSphere
      sphereCollisionRadius=1
      boundBone {boneName=hair_05}
      boxExtents=<10 5 5>
      constraintSetup {angularLimitsMax=<5 10 60>, angularLimitsMin=<-5 -10 -5>, angularTarget=<0 0 0>}
      localJointOffset=<0 0 1>
     }]
   }
   n10 = Anim Dynamics - Bone: hair_06(ComponentPose=@n9, Alpha=1, ExternalForce="0,0,0") {
    angularDampingOverride=0.99
    angularSpringConstant=500
    bOverrideAngularDamping=true
    bUseGravityOverride=true
    bUsePlanarLimit=false
    bUseSphericalLimits=true
    gravityScale=-1
    simulationSpace=RootRelative
    alphaBoolBlend {blendOption=Linear}
    boundBone {boneName=hair_06}
    componentAppliedLinearAccClamp=<5000 5000 5000>
    componentLinearAccScale=<0.2 0.2 0.2>
    componentLinearVelScale=<0.2 0.2 0.2>
    gravityOverride {z=-980}
    physicsBodyDefinitions [{
      collisionType=CustomSphere
      sphereCollisionRadius=1
      boundBone {boneName=hair_06}
      boxExtents=<10 5 5>
      constraintSetup {angularLimitsMax=<5 10 60>, angularLimitsMin=<-5 -10 -5>, angularTarget=<0 0 0>}
      localJointOffset=<0 0 1>
     }]
   }
   n11 = Anim Dynamics - Bone: hair_07(ComponentPose=@n10, Alpha=1, ExternalForce="0,0,0") {
    angularDampingOverride=0.99
    angularSpringConstant=500
    bOverrideAngularDamping=true
    bUseGravityOverride=true
    bUsePlanarLimit=false
    bUseSphericalLimits=true
    gravityScale=-1
    simulationSpace=RootRelative
    alphaBoolBlend {blendOption=Linear}
    boundBone {boneName=hair_07}
    componentAppliedLinearAccClamp=<5000 5000 5000>
    componentLinearAccScale=<0.2 0.2 0.2>
    componentLinearVelScale=<0.2 0.2 0.2>
    gravityOverride {z=-980}
    physicsBodyDefinitions [{
      collisionType=CustomSphere
      sphereCollisionRadius=1
      boundBone {boneName=hair_07}
      boxExtents=<10 5 5>
      constraintSetup {angularLimitsMax=<5 10 60>, angularLimitsMin=<-5 -10 -5>, angularTarget=<0 0 0>}
      localJointOffset=<0 0 1>
     }]
   }
   Output Pose(Result=Component To Local(ComponentPose=@n11)) {layerGroup=None}
  ]}
 EventGraph {nodes [
   Event BlueprintUpdateAnimation()
   n1 = Get CurrentLOD()
   > Set Is LOD 1(Is LOD 1:bool="Equal (Integer)"(A=@n1, B=1)) -> (Output_Get:bool)
   > Set Is LOD 2(Is LOD 2:bool="integer >= integer"(A=@n1, B=2)) -> (Output_Get:bool)
  ]}
]
variables [
 Is LOD 1 {type=boolean}
 Is LOD 2 {type=boolean}
]
